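; Dino rift encounter Blueprint, a child class that holds back its creature spawning until shortly after the encounter spawns.
; Its only active logic runs on post-spawn events: it calls the parent, waits a short delay, then allows the spawner to spawn.
; An author note says the delay exists because animals were often seen spawning before the "Contained", likely meaning their container.
; BeginPlay, overlap, Tick and the construction script add no logic of their own.
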(event EventBeginPlay
  (|CallParentFunction) ; This node is disabled and will not be called.\nDrag off pins to build functionality.
)

(event Collision|EventActorBeginOverlap
  (params (OtherActor :type Actor%20Object%20Reference)))

(event EventTick
  (params (DeltaSeconds :type Float%20%28single-precision%29)))

(event EventPostSpawnEvents
  ;[1 1 1] Since we often see the Animals Spawn beforr the Contained, make sure we spawn them a bit later.
  (|CallParentFunction)
  (Utilities|FlowControl|Delay 1.0)
  (CallFunction|AllowSpawnertoSpawn))

(fn ConstructionScript
  (|CallParentFunction))
